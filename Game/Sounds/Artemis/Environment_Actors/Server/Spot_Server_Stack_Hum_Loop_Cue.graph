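# Looping positional hum cue for server stack environment props, with its own attenuation and a low voice-count concurrency override.
# A random node picks one of six looping hum waves, each pitched down by a modulator and slowly wavered by an indefinitely looping pitch envelope.
# A short random start delay and an outer fade-in envelope wrap the pick.
# The quality node routes only its first two slots to this chain, leaving the other two empty.
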SoundCue /Game/Sounds/Artemis/Environment_Actors/Server/Spot_Server_Stack_Hum_Loop_Cue
root=QualityLevel_0
nodes [
 {id=Delay_0, children [Random_0], properties {delayMax=0.1, delayMin=0.05}}
 {
  id=Enveloper_22
  children [Modulator_2]
  properties {
   bLoop=true
   bLoopIndefinitely=true
   loopEnd=60
   pitchCurve {editorCurveData {keys [{interpMode=RCIM_Cubic, time=0}, <at 30 0.95 RCIM_Cubic>, {interpMode=RCIM_Cubic, time=60}]}}
   volumeCurve {editorCurveData {keys [<at 0 1>]}}
  }
 }
 {
  id=Enveloper_23
  children [Modulator_3]
  properties {
   bLoop=true
   bLoopIndefinitely=true
   loopEnd=60
   pitchCurve {editorCurveData {keys [{interpMode=RCIM_Cubic, time=0}, <at 30 0.95 RCIM_Cubic>, {interpMode=RCIM_Cubic, time=60}]}}
   volumeCurve {editorCurveData {keys [<at 0 1>]}}
  }
 }
 {
  id=Enveloper_24
  children [Modulator_4]
  properties {
   bLoop=true
   bLoopIndefinitely=true
   loopEnd=60
   pitchCurve {editorCurveData {keys [{interpMode=RCIM_Cubic, time=0}, <at 30 0.95 RCIM_Cubic>, {interpMode=RCIM_Cubic, time=60}]}}
   volumeCurve {editorCurveData {keys [<at 0 1>]}}
  }
 }
 {
  id=Enveloper_25
  children [Modulator_5]
  properties {
   bLoop=true
   bLoopIndefinitely=true
   loopEnd=60
   pitchCurve {editorCurveData {keys [{interpMode=RCIM_Cubic, time=0}, <at 30 0.95 RCIM_Cubic>, {interpMode=RCIM_Cubic, time=60}]}}
   volumeCurve {editorCurveData {keys [<at 0 1>]}}
  }
 }
 {
  id=Enveloper_26
  children [Modulator_6]
  properties {
   bLoop=true
   bLoopIndefinitely=true
   loopEnd=60
   pitchCurve {editorCurveData {keys [{interpMode=RCIM_Cubic, time=0}, <at 30 0.95 RCIM_Cubic>, {interpMode=RCIM_Cubic, time=60}]}}
   volumeCurve {editorCurveData {keys [<at 0 1>]}}
  }
 }
 {
  id=Enveloper_27
  children [Modulator_1]
  properties {
   bLoop=true
   bLoopIndefinitely=true
   loopEnd=60
   pitchCurve {editorCurveData {keys [{interpMode=RCIM_Cubic, time=0}, <at 30 0.95 RCIM_Cubic>, {interpMode=RCIM_Cubic, time=60}]}}
   volumeCurve {editorCurveData {keys [<at 0 1>]}}
  }
 }
 {
  id=Enveloper_29
  children [Delay_0]
  properties {
   pitchCurve {editorCurveData {keys [{time=0}]}}
   volumeCurve {editorCurveData {keys [{arriveTangent=3.4293, interpMode=RCIM_Cubic, leaveTangent=3.4293, tangentMode=RCTM_User, time=0, value=0}, <at 0.4 1 RCIM_Cubic>]}}
  }
 }
 {id=Modulator_1, children [WavePlayer_2], properties {pitchMax=0.9, pitchMin=0.8}}
 {id=Modulator_2, children [WavePlayer_3], properties {pitchMax=0.9, pitchMin=0.8}}
 {id=Modulator_3, children [WavePlayer_4], properties {pitchMax=1, pitchMin=0.9}}
 {id=Modulator_4, children [WavePlayer_5], properties {pitchMax=0.9, pitchMin=0.8, volumeMax=0.85, volumeMin=0.8}}
 {id=Modulator_5, children [WavePlayer_6], properties {pitchMax=0.9, pitchMin=0.8}}
 {id=Modulator_6, children [WavePlayer_7], properties {pitchMax=0.9, pitchMin=0.8}}
 {id=QualityLevel_0, children [Enveloper_29,Enveloper_29,null,null]}
 {id=Random_0, children [Enveloper_27,Enveloper_22,Enveloper_23,Enveloper_24,Enveloper_25,Enveloper_26], properties {weights [1,1,1,1,1,1]}}
 {id=WavePlayer_2, properties {bLooping=true, soundWaveAssetPtr=&/Spot_Server_Stack_Hum_02}}
 {id=WavePlayer_3, properties {bLooping=true, soundWaveAssetPtr=&/Spot_Server_Stack_Hum_03}}
 {id=WavePlayer_4, properties {bLooping=true, soundWaveAssetPtr=&/Spot_Server_Stack_Hum_04}}
 {id=WavePlayer_5, properties {bLooping=true, soundWaveAssetPtr=&/Spot_Server_Stack_Hum_05}}
 {id=WavePlayer_6, properties {bLooping=true, soundWaveAssetPtr=&/Spot_Server_Stack_Hum_06}}
 {id=WavePlayer_7, properties {bLooping=true, soundWaveAssetPtr=&/Spot_Server_Stack_Hum_07}}
]
properties {
 attenuationSettings=&/Spot_Server_Stack_Attenuation
 bEnableBusSends=false
 bEnableSubmixSends=false
 bOverrideConcurrency=true
 maxDistance=562
 soundClassObject=&Fort_Gadgets
 volumeMultiplier=0.3
 concurrencyOverrides {maxCount=2, voiceStealReleaseTime=0.2}
}
